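# Masked, Nanite-enabled Default Lit material for dust pile props, going by its name, in the Transylvania environment set.
# A scalar Tiling parameter scales the UVs shared by the D, N and S texture parameters, which default to white rock and snow pile textures.
# The D texture is multiplied by a tint color that defaults to black.
# Vertex color feeds DitherTemporalAA, giving a painted, dithered threshold suited to soft masked edges.
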Material /Game/Environments/Props/Transylvania/Props/Materials/M_Dust_piles
graph {
 properties {bUsedWithNanite=true, blendMode=BLEND_Masked, pixelDepthOffsetMode=PDOM_Legacy, usedShadingModels=Default Lit}
 nodes [
  MaterialFunctionCall(Alpha Threshold=VertexColor()) {materialFunction=&DitherTemporalAA}
  n1 = ScalarParameter() {defaultValue=1, parameterName=Tiling}
  n2 = Multiply(A=TextureCoordinate(), B=AppendVector(A=@n1, B=@n1))
  Multiply(
   A=VectorParameter() {parameterName=Tint_Base_color, defaultValue=<rgba 0 0 0 1>}.RGB
   B=TextureSampleParameter2D(UVs=@n2) {parameterName=D, texture=/Game/Packages/DS_Fortnight/WhiteRockCustom_D}.RGB
  )
  TextureSampleParameter2D(UVs=@n2) {parameterName=N, samplerType=SAMPLERTYPE_Normal, texture=/Game/Packages/DS_Fortnight/WhiteRockCustom_N}
  TextureSampleParameter2D(UVs=@n2) {parameterName=S, samplerType=SAMPLERTYPE_Masks, texture=/Game/Environments/Props/Winter/Textures/T_SnowPile_S}
 ]
}
